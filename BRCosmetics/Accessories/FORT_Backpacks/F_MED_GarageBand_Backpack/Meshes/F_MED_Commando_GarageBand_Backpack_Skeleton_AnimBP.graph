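# Animation Blueprint for the GarageBand backpack cosmetic (medium female Commando) that procedurally offsets the backpack_root bone of an incoming pose.
# Three chained Transform Bone nodes apply a variable translation and additive rotation, then fixed additive rotations weighted by BackpackRotationAlpha and SwingRotationAlpha.
# Both alphas use interpolation for smoothing.
# The driving variables are not declared here, and the update event has no logic.
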AnimBlueprint /BRCosmetics/Accessories/FORT_Backpacks/F_MED_GarageBand_Backpack/Meshes/F_MED_Commando_GarageBand_Backpack_Skeleton_AnimBP
target_skeleton=&/F_MED_Commando_GarageBand_Backpack_Skeleton
graphs [
 AnimGraph {nodes [
   AnimGraphNode_SubInput_0 = Input Pose()
   n1 = "Transform (Modify) Bone - Bone: backpack_root"(
    ComponentPose=Local To Component(LocalPose=@AnimGraphNode_SubInput_0)
    Alpha=1
    Translation=Get BackpackTranslation() -> (BackpackTranslation:Vector)
    Rotation=Get BackpackRotator() -> (BackpackRotator:Rotator)
    Scale="1,1,1"
   ) {
    rotationMode=BMM_Additive
    rotationSpace=BCS_BoneSpace
    translationMode=BMM_Replace
    alphaBoolBlend {blendOption=Linear}
    boneToModify {boneName=backpack_root}
   }
   n2 = "Transform (Modify) Bone - Bone: backpack_root"(ComponentPose=@n1, Alpha=Get BackpackRotationAlpha() -> (BackpackRotationAlpha:float), Translation="0,0,0", Rotation="5,20,0", Scale="1,1,1") {
    rotationMode=BMM_Additive
    rotationSpace=BCS_BoneSpace
    alphaBoolBlend {blendOption=Linear}
    alphaScaleBiasClamp {bInterpResult=true, interpSpeedDecreasing=20, interpSpeedIncreasing=20}
    boneToModify {boneName=backpack_root}
    rotation {pitch=5, yaw=20}
   }
   n3 = "Transform (Modify) Bone - Bone: backpack_root"(ComponentPose=@n2, Alpha=Get SwingRotationAlpha() -> (SwingRotationAlpha:float), Translation="0,0,0", Rotation="50,0,20", Scale="1,1,1") {
    rotationMode=BMM_Additive
    rotationSpace=BCS_BoneSpace
    alphaBoolBlend {blendOption=Linear}
    alphaScaleBiasClamp {bInterpResult=true, interpSpeedDecreasing=5}
    boneToModify {boneName=backpack_root}
    rotation {pitch=50, roll=20}
   }
   Output Pose(Result=Component To Local(ComponentPose=@n3)) {layerGroup=None}
  ]}
 EventGraph {nodes [Event BlueprintUpdateAnimation()]}
]
variables []
